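# Masked, dithered, two-sided metallic material for the Witch glider, giving it animated blue emissive glow and sparkles.
# Panning fire noise, offset by the desaturated emissive texture, drives a three-color Fresnel gradient and warps small-noise sparkles.
# A UV2 gradient height-lerped with that noise and texture forms the opacity mask, with extra gradient fade on back faces.
# A panning swirl normal blends over the base normal except on mobile, and attributes pass through the accessories tech-art override function.
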Material /Game/Gadgets/Assets/VinderTech_GliderChute/Glider_Witch/Materials/M_Glider_Witch
graph {
 properties {
  bUseMaterialAttributes=true
  bUsedWithMeshDeformer=true
  bUsedWithSkeletalMesh=true
  blendMode=BLEND_Masked
  ditherOpacityMask=true
  pixelDepthOffsetMode=PDOM_Legacy
  translucencyLightingMode=TLM_Surface
  twoSided=true
  usedShadingModels=Default Lit
 }
 nodes [
  n1 = Constant() {r=5}
  n2 = Multiply(
   A=TextureCoordinate() {coordinateIndex=1}
   B=MaterialFunctionCall(X=ScalarParameter() {defaultValue=2.5, parameterName=IceTiling_11}, Y=ScalarParameter() {defaultValue=1.5, parameterName=IceTiling_10}) {materialFunction=&MakeFloat2}.Result
  )
  n3 = Multiply(A=Time(), B=ScalarParameter() {defaultValue=0.25, parameterName=IceTiling_5})
  n4 = Desaturation(
   None=TextureSample() {samplerType=SAMPLERTYPE_LinearColor, texture=/Game/Gadgets/Assets/VinderTech_GliderChute/Glider_Witch/Textures/T_Glider_Witch_E}.RGB
   Fraction=Constant() {r=1}
  )
  n5 = Power(Base=@n4, Exp=ScalarParameter() {defaultValue=4, parameterName=BaseDiffuseMaskPower})
  n6 = TextureSample(UVs=Add(A=Panner(Coordinate=@n2, Time=@n3) {speedX=-0.5, speedY=-1}, B=@n5)) {texture=/Game/Global/Textures/Default/Noise/T_FireNoise_02}
  n7 = Multiply(A=@n6.R, B=ScalarParameter() {defaultValue=-0.1, parameterName=SparklesWarpAmount})
  n8 = Multiply(A=TextureCoordinate() {coordinateIndex=1, uTiling=1.5}, B=ScalarParameter() {defaultValue=1.2, parameterName=IceTiling_8})
  n9 = OneMinus(None=TextureSample(UVs=Add(A=@n7, B=@n8)) {samplerType=SAMPLERTYPE_Masks, texture=/Game/Characters/Fortress_CharacterShader/SmallNoiseTexture}.RGB)
  n10 = Abs(None=Fresnel(ExponentIn=ScalarParameter() {defaultValue=1.5, parameterName=Fresnel-Expo}) {baseReflectFraction=0})
  n11 = Multiply(
   A=Multiply(
    A=MaterialFunctionCall(Transition Phase=@n6.R, Height Texture=@n9) {materialFunction=&HeightLerp}.Lerp Alpha No Contrast
    B=@n9
   )
   B=@n10
  )
  n12 = Power(Base=@n6.R, Exp=ScalarParameter() {defaultValue=4, parameterName=PanningMask-Power})
  n13 = Multiply(A=@n5, B=ScalarParameter() {defaultValue=3, parameterName=BaseDiffuse-BrightMod})
  n14 = LinearInterpolate(
   A=LinearInterpolate(
    A=VectorParameter() {parameterName=Color-01, defaultValue=<rgba 0 0.0337 0.3 1>}.RGB
    B=VectorParameter() {parameterName=Color-02, defaultValue=<rgba 0 0.1123 1 1>}.RGB
    Alpha=@n12
   )
   B=VectorParameter() {parameterName=Color-03, defaultValue=<rgba 0.3 1.0734 3 1>}.RGB
   Alpha=Multiply(A=Saturate(None=@n13), B=@n10)
  )
  n15 = ScalarParameter() {defaultValue=4, parameterName=MainFres-Bright}
  n16 = Power(Base=@n4, Exp=ScalarParameter() {defaultValue=2, parameterName=BaseDiffuseMaskPower_1})
  n17 = ScalarParameter() {defaultValue=0.5, parameterName=BaseDiffuseMaskPower_2}
  n18 = LinearInterpolate(
   A=Multiply(A=ScalarParameter() {defaultValue=0.5, parameterName=MainFres-Dark}, B=@n14)
   B=Multiply(A=Multiply(A=@n15, B=@n14), B=Add(A=@n16, B=@n17))
   Alpha=Power(Base=OneMinus(None=@n10), Exp=ScalarParameter() {defaultValue=50, parameterName=MainFres-Power})
  )
  Add(A=@n18, B=Multiply(A=@n18, B=Multiply(A=@n1, B=@n11)))
  Constant() {r=2}
  Constant() {r=25}
  Constant() {r=3}
  Constant() {r=0.5}
  Constant() {r=3}
  Constant() {r=0.15}
  n19 = ComponentMask(None=TextureCoordinate() {coordinateIndex=2}) {g=true}
  n20 = MaterialFunctionCall(Transition Phase=Power(Base=@n19, Exp=ScalarParameter() {defaultValue=1.25, parameterName=BDM-Power_1}), Height Texture=Add(A=@n12, B=@n4)) {materialFunction=&HeightLerp}
  n21 = Multiply(A=@n5, B=ScalarParameter() {defaultValue=1, parameterName=RearEdgesAmount})
  n22 = Saturate(None=Add(A=Add(A=@n20.Lerp Alpha No Contrast, B=@n21), B=ScalarParameter() {defaultValue=0.4, parameterName=LowestOpacityAmount}))
  n23 = Multiply(A=@n19, B=@n22)
  n24 = Saturate(None=TwoSidedSign())
  n25 = TextureSample() {samplerType=SAMPLERTYPE_Normal, texture=/Game/Gadgets/Assets/VinderTech_GliderChute/Glider_Witch/Textures/T_Glider_Witch_N}
  n26 = Multiply(
   A=TextureCoordinate()
   B=MaterialFunctionCall(X=ScalarParameter() {defaultValue=0.4, parameterName=IceTiling_4}, Y=ScalarParameter() {defaultValue=0.6, parameterName=IceTiling_3}) {materialFunction=&MakeFloat2}.Result
  )
  n27 = Multiply(A=Time(), B=ScalarParameter() {defaultValue=0.12, parameterName=IceTiling_6})
  n28 = LinearInterpolate(
   A=TextureSample(UVs=Panner(Coordinate=@n26, Time=@n27) {speedX=-0.5, speedY=-1}) {samplerType=SAMPLERTYPE_Normal, texture=/Game/Global/Textures/Default/Noise/T_SwirlNoise_N}.RGB
   B=Constant3Vector() {constant=<rgba 0 0 1 0>}
   Alpha=ScalarParameter() {defaultValue=0.25, parameterName=TilingNormalAmount}
  )
  n29 = SetMaterialAttributes!(
   MaterialAttributes
   Base Color=VectorParameter() {parameterName=BaseColor, defaultValue=<rgba 0 0.0123 0.1 1>}.RGB
   Metallic=Constant() {r=1}
   Roughness=ScalarParameter() {defaultValue=0.3, parameterName=Rough}
   Emissive Color=LinearInterpolate(A=@n18, B=VectorParameter() {parameterName=SparklesColor, defaultValue=<rgba 0.125 0.6335 2.5 1>}.RGB, Alpha=@n11)
   Opacity Mask=Add(A=LinearInterpolate(A=@n23, B=@n22, Alpha=@n24), B=ScalarParameter() {parameterName=AddToOpacity})
   Normal=ShadingPathSwitch(
    Default=MaterialFunctionCall(BaseNormal=@n25.RGB, AdditionalNormal=@n28) {materialFunction=&BlendAngleCorrectedNormals}.Result
    Mobile=@n25.RGB
   ) {desc=Migrated from FeatureLevelSwitch}
  ) {attributeSetTypes [
    #1
    #2
    #3
    #4
    #5
    #6
   ]}
  MaterialFunctionCall(In=@n29) {materialFunction=&MF_TechArtOverrides_accessories}
  Multiply(A=ScalarParameter() {defaultValue=1, parameterName=WarpPanningMask-Mod}, B=@n5)
  Multiply(
   A=TextureCoordinate() {coordinateIndex=1}
   B=MaterialFunctionCall(X=Constant() {r=1.5}, Y=Constant() {r=1}) {materialFunction=&MakeFloat2}.Result
  )
 ]
}
